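# Physics asset for the Scarecrow female medium body cosmetic, covering the first left and right dynamic skirt bones.
# Each skirt bone has a large capsule offset along X in mirrored directions, while each thigh has only a tiny capsule.
# Two constraints pair each skirt bone with the thigh on its side, using narrow, equal swing and twist limit angles.
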PhysicsAsset /BRCosmetics/Characters/Player/Female/Medium/Bodies/F_MED_Scarecrow/Meshes/F_MED_Scarecrow_Body_Dyn_Physics
bodies {
 dyn_skirt_01_l [{shapeType=Capsule, length=16, radius=8, center=<-8 0 0>, rotation=<rot 90 0 0>}]
 dyn_skirt_01_r [{shapeType=Capsule, length=16, radius=8, center=<8 0 0>, rotation=<rot 90 0 0>}]
 thigh_l [{shapeType=Capsule, length=1, radius=1}]
 thigh_r [{shapeType=Capsule, length=1, radius=1}]
}
constraints [
 {bone1Name=dyn_skirt_01_l, bone2Name=thigh_l, swing1LimitDegrees=10, swing2LimitDegrees=10, twistLimitDegrees=10}
 {bone1Name=dyn_skirt_01_r, bone2Name=thigh_r, swing1LimitDegrees=10, swing2LimitDegrees=10, twistLimitDegrees=10}
]
