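# Post-process style AnimBP for the female Scarecrow player body, adding secondary motion to ribbons, necklace, collar and skirt.
# The pose copied from the parent mesh runs through Anim Dynamics and ribbon rotation constraints.
# Right-hand distance to the skirt drives an additive skirt bend and the alpha of two right-skirt simulations.
# The update event also detects backbling to set a material hide parameter; thigh-follow, RigidBody and spring nodes are unconnected.
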AnimBlueprint /BRCosmetics/Characters/Player/Female/Medium/Bodies/F_MED_Scarecrow/Meshes/F_MED_Scarecrow_Body_AnimBP
target_skeleton=&/F_MED_Scarecrow_Body_Skeleton
graphs [
 AnimGraph {nodes [
   n1 = "Transform (Modify) Bone - Bone: dyn_skirt_01_l"(Alpha=1, Translation="0,0,0", Rotation=Get left_thigh_rotation() -> (left_thigh_rotation:Rotator), Scale="1,1,1") {rotationMode=BMM_Additive, rotationSpace=BCS_BoneSpace, alphaBoolBlend {blendOption=Linear}, boneToModify {boneName=dyn_skirt_01_l}}
   "Transform (Modify) Bone - Bone: dyn_skirt_01_r"(ComponentPose=@n1, Alpha=1, Translation="0,0,0", Rotation=Get right_thigh_rotation() -> (right_thigh_rotation:Rotator), Scale="1,1,1") {rotationMode=BMM_Additive, rotationSpace=BCS_BoneSpace, alphaBoolBlend {blendOption=Linear}, boneToModify {boneName=dyn_skirt_01_r}}
   RigidBody() {
    bOverrideWorldGravity=true
    overridePhysicsAsset=&/F_MED_Scarecrow_Body_Dyn_Physics
    alphaBoolBlend {blendOption=Linear}
    overrideWorldGravity {z=-980}
   }
   n2 = Anim Dynamics - Bone: dyn_back_ribbon_02_r(ComponentPose=Local To Component(LocalPose=Copy Pose From Mesh() {bUseAttachedParent=true}), Alpha=0.5, ExternalForce="0,0,0") {
    angularDampingOverride=0.99
    angularSpringConstant=600
    bAngularSpring=true
    bGravityOverrideInSimSpace=true
    bOverrideAngularDamping=true
    bUseGravityOverride=true
    gravityScale=0.01
    alphaBoolBlend {blendOption=Linear}
    boundBone {boneName=dyn_back_ribbon_02_r}
    componentLinearAccScale=<0.5 0.5 0.5>
    componentLinearVelScale=<0.5 0.5 0.5>
    gravityOverride {z=-500}
    physicsBodyDefinitions [{
      sphereCollisionRadius=0
      boundBone {boneName=dyn_back_ribbon_02_r}
      boxExtents=<5 5 2>
      constraintSetup {angularTargetAxis=AxisZ, twistAxis=AxisZ, angularLimitsMax=<0 30 0>, angularLimitsMin=<-120 -30 0>, angularTarget=<0 0 0>}
      localJointOffset=<0 0 -10>
     }]
   }
   n3 = Anim Dynamics - Bone: dyn_back_ribbon_02_l(ComponentPose=@n2, Alpha=0.5, ExternalForce="0,0,0") {
    angularDampingOverride=0.99
    angularSpringConstant=600
    bAngularSpring=true
    bGravityOverrideInSimSpace=true
    bOverrideAngularDamping=true
    bUseGravityOverride=true
    gravityScale=0.01
    alphaBoolBlend {blendOption=Linear}
    boundBone {boneName=dyn_back_ribbon_02_l}
    componentLinearAccScale=<0.5 0.5 0.5>
    componentLinearVelScale=<0.5 0.5 0.5>
    gravityOverride {z=-500}
    physicsBodyDefinitions [{
      sphereCollisionRadius=0
      boundBone {boneName=dyn_back_ribbon_02_l}
      boxExtents=<5 5 2>
      constraintSetup {angularTargetAxis=AxisZ, twistAxis=AxisZ, angularLimitsMax=<0 30 0>, angularLimitsMin=<-120 -30 0>, angularTarget=<0 0 0>}
      localJointOffset=<0 0 -10>
     }]
   }
   n4 = "Constraint - dyn_back_ribbon_01_r (R)"(ComponentPose=@n3, Alpha=0.5, ConstraintWeights_0=1) {
    alphaBoolBlend {blendOption=Linear}
    boneToModify {boneName=dyn_back_ribbon_01_r}
    constraintSetup [{transformType=Rotation, targetBone {boneName=dyn_back_ribbon_02_r}}]
    constraintWeights [1]
   }
   n5 = "Constraint - dyn_back_ribbon_01_l (R)"(ComponentPose=@n4, Alpha=0.5, ConstraintWeights_0=1) {
    alphaBoolBlend {blendOption=Linear}
    boneToModify {boneName=dyn_back_ribbon_01_l}
    constraintSetup [{transformType=Rotation, targetBone {boneName=dyn_back_ribbon_02_l}}]
    constraintWeights [1]
   }
   n6 = Anim Dynamics - Bone: dyn_necklace(ComponentPose=@n5, Alpha=1, ExternalForce="0,0,0") {
    angularDampingOverride=0.99
    angularSpringConstant=600
    bAngularSpring=true
    bGravityOverrideInSimSpace=true
    bOverrideAngularDamping=true
    bUseGravityOverride=true
    gravityScale=0.01
    alphaBoolBlend {blendOption=Linear}
    boundBone {boneName=dyn_necklace}
    componentLinearAccScale=<0.1 0.1 0.1>
    componentLinearVelScale=<0.1 0.1 0.1>
    gravityOverride {z=-500}
    physicsBodyDefinitions [{
      sphereCollisionRadius=0
      boundBone {boneName=dyn_necklace}
      boxExtents=<10 10 5>
      constraintSetup {angularTargetAxis=AxisZ, twistAxis=AxisZ, angularLimitsMax=<20 45 0>, angularLimitsMin=<-20 10 0>, angularTarget=<0 0 0>}
      localJointOffset=<0 0 -5>
     }]
   }
   n7 = Anim Dynamics - Bone: dyn_collar(ComponentPose=@n6, Alpha=1, ExternalForce="0,0,0") {
    angularDampingOverride=0.99
    angularSpringConstant=1200
    bAngularSpring=true
    bGravityOverrideInSimSpace=true
    bOverrideAngularDamping=true
    bUseGravityOverride=true
    gravityScale=0.01
    alphaBoolBlend {blendOption=Linear}
    boundBone {boneName=dyn_collar}
    componentLinearAccScale=<0.1 0.1 0.1>
    componentLinearVelScale=<0.1 0.1 0.1>
    gravityOverride {z=-500}
    physicsBodyDefinitions [{
      sphereCollisionRadius=0
      boundBone {boneName=dyn_collar}
      boxExtents=<20 20 10>
      constraintSetup {angularTargetAxis=AxisZ, twistAxis=AxisZ, angularLimitsMax=<20 0 0>, angularLimitsMin=<-20 0 0>, angularTarget=<0 0 0>}
      localJointOffset=<0 0 -5>
     }]
   }
   n8 = Get hand_l_dist_to_skirt() -> (hand_l_dist_to_skirt:double)
   n9 = "Transform (Modify) Bone - Bone: dyn_skirt_B_01_r"(ComponentPose=@n7, Alpha=@n8, Translation="0,0,0", Rotation="40,0.000,0.000", Scale="1,1,1") {rotationMode=BMM_Additive, alphaBoolBlend {blendOption=Linear}, boneToModify {boneName=dyn_skirt_B_01_r}, rotation {pitch=40}}
   n10 = Anim Dynamics - Bone: dyn_skirt_A_01_l(ComponentPose=@n9, Alpha=1, ExternalForce="0,0,0") {
    angularDampingOverride=0.98
    angularSpringConstant=500
    bAngularSpring=true
    bGravityOverrideInSimSpace=true
    bOverrideAngularDamping=true
    bUseGravityOverride=true
    gravityScale=0.05
    simulationSpace=RootRelative
    alphaBoolBlend {blendOption=Linear}
    boundBone {boneName=dyn_skirt_A_01_l}
    componentLinearAccScale=<1 1 1>
    componentLinearVelScale=<1 1 1>
    gravityOverride {z=-980}
    physicsBodyDefinitions [{
      sphereCollisionRadius=0
      boundBone {boneName=dyn_skirt_A_01_l}
      boxExtents=<10 20 20>
      constraintSetup {angularLimitsMax=<0 0 -5>, angularLimitsMin=<0 0 -40>, angularTarget=<0 0 0>}
      localJointOffset=<10 0 0>
     }]
   }
   n11 = Anim Dynamics - Bone: dyn_skirt_B_01_l(ComponentPose=@n10, Alpha=1, ExternalForce="0,0,0") {
    angularDampingOverride=0.98
    angularSpringConstant=500
    bAngularSpring=true
    bGravityOverrideInSimSpace=true
    bOverrideAngularDamping=true
    bUseGravityOverride=true
    gravityScale=0.05
    simulationSpace=RootRelative
    alphaBoolBlend {blendOption=Linear}
    boundBone {boneName=dyn_skirt_B_01_l}
    componentLinearAccScale=<1 1 1>
    componentLinearVelScale=<1 1 1>
    gravityOverride {z=-980}
    physicsBodyDefinitions [{
      sphereCollisionRadius=0
      boundBone {boneName=dyn_skirt_B_01_l}
      boxExtents=<10 20 20>
      constraintSetup {angularLimitsMax=<0 20 10>, angularLimitsMin=<0 -20 -40>, angularTarget=<0 0 0>}
      localJointOffset=<10 0 0>
     }]
   }
   n12 = Anim Dynamics - Bone: dyn_skirt_C_01_l(ComponentPose=@n11, Alpha=1, ExternalForce="0,0,0") {
    angularDampingOverride=0.98
    angularSpringConstant=500
    bAngularSpring=true
    bGravityOverrideInSimSpace=true
    bOverrideAngularDamping=true
    bUseGravityOverride=true
    gravityScale=0.05
    simulationSpace=RootRelative
    alphaBoolBlend {blendOption=Linear}
    boundBone {boneName=dyn_skirt_C_01_l}
    componentLinearAccScale=<1 1 1>
    componentLinearVelScale=<1 1 1>
    gravityOverride {z=-980}
    physicsBodyDefinitions [{
      sphereCollisionRadius=0
      boundBone {boneName=dyn_skirt_C_01_l}
      boxExtents=<10 20 20>
      constraintSetup {angularLimitsMax=<0 20 10>, angularLimitsMin=<0 -20 -40>, angularTarget=<0 0 0>}
      localJointOffset=<10 0 0>
     }]
   }
   n13 = float - float(B=@n8) {is=title K2Node_CallFunction}
   n14 = Anim Dynamics - Bone: dyn_skirt_A_01_r(ComponentPose=@n12, Alpha=@n13, ExternalForce="0,0,0") {
    angularDampingOverride=0.98
    angularSpringConstant=500
    bAngularSpring=true
    bGravityOverrideInSimSpace=true
    bOverrideAngularDamping=true
    bUseGravityOverride=true
    gravityScale=0.05
    simulationSpace=RootRelative
    alphaBoolBlend {blendOption=Linear}
    alphaScaleBiasClamp {bInterpResult=true, interpSpeedIncreasing=5}
    boundBone {boneName=dyn_skirt_A_01_r}
    componentLinearAccScale=<1 1 1>
    componentLinearVelScale=<1 1 1>
    gravityOverride {z=-980}
    physicsBodyDefinitions [{
      sphereCollisionRadius=0
      boundBone {boneName=dyn_skirt_A_01_r}
      boxExtents=<10 20 20>
      constraintSetup {angularLimitsMax=<0 0 -5>, angularLimitsMin=<0 0 -40>, angularTarget=<0 0 0>}
      localJointOffset=<10 0 0>
     }]
   }
   n15 = Anim Dynamics - Bone: dyn_skirt_B_01_r(ComponentPose=@n14, Alpha=@n13, ExternalForce="0,0,0") {
    angularDampingOverride=0.98
    angularSpringConstant=500
    bAngularSpring=true
    bGravityOverrideInSimSpace=true
    bOverrideAngularDamping=true
    bUseGravityOverride=true
    gravityScale=0.05
    simulationSpace=RootRelative
    alphaBoolBlend {blendOption=Linear}
    alphaScaleBiasClamp {bInterpResult=true, interpSpeedIncreasing=5}
    boundBone {boneName=dyn_skirt_B_01_r}
    componentLinearAccScale=<1 1 1>
    componentLinearVelScale=<1 1 1>
    gravityOverride {z=-980}
    physicsBodyDefinitions [{
      sphereCollisionRadius=0
      boundBone {boneName=dyn_skirt_B_01_r}
      boxExtents=<10 20 20>
      constraintSetup {angularLimitsMax=<0 20 10>, angularLimitsMin=<0 -20 -40>, angularTarget=<0 0 0>}
      localJointOffset=<10 0 0>
     }]
   }
   n16 = Anim Dynamics - Bone: dyn_skirt_C_01_r(ComponentPose=@n15, Alpha=1, ExternalForce="0,0,0") {
    angularDampingOverride=0.98
    angularSpringConstant=500
    bAngularSpring=true
    bGravityOverrideInSimSpace=true
    bOverrideAngularDamping=true
    bUseGravityOverride=true
    gravityScale=0.05
    simulationSpace=RootRelative
    alphaBoolBlend {blendOption=Linear}
    boundBone {boneName=dyn_skirt_C_01_r}
    componentLinearAccScale=<1 1 1>
    componentLinearVelScale=<1 1 1>
    gravityOverride {z=-980}
    physicsBodyDefinitions [{
      sphereCollisionRadius=0
      boundBone {boneName=dyn_skirt_C_01_r}
      boxExtents=<10 20 20>
      constraintSetup {angularLimitsMax=<0 20 10>, angularLimitsMin=<0 -20 -40>, angularTarget=<0 0 0>}
      localJointOffset=<10 0 0>
     }]
   }
   Output Pose(Result=Component To Local(ComponentPose=@n16)) {layerGroup=None}
   n17 = Spring controller - Bone: dyn_skirt_01_l!(ComponentPose:Component Space Pose Link Structure, Alpha:float=1) -> (Pose:Component Space Pose Link Structure) {
    is=title AnimGraphNode_SpringBone
    bRotateX=true
    bRotateY=true
    bRotateZ=true
    maxDisplacement=10
    springDamping=32
    springStiffness=1500
    alphaBoolBlend {blendOption=Linear}
    springBone {boneName=dyn_skirt_01_l}
   }
   Spring controller - Bone: dyn_skirt_01_r!(ComponentPose:Component Space Pose Link Structure=@n17, Alpha:float=1) -> (Pose:Component Space Pose Link Structure) {
    is=title AnimGraphNode_SpringBone
    bRotateX=true
    bRotateY=true
    bRotateZ=true
    maxDisplacement=10
    springDamping=32
    springStiffness=1500
    alphaBoolBlend {blendOption=Linear}
    springBone {boneName=dyn_skirt_01_r}
   }
   Spring controller - Bone: dyn_skirt_global!(ComponentPose:Component Space Pose Link Structure, Alpha:float=1) -> (Pose:Component Space Pose Link Structure) {
    is=title AnimGraphNode_SpringBone
    bRotateZ=true
    maxDisplacement=5
    alphaBoolBlend {blendOption=Linear}
    springBone {boneName=dyn_skirt_global}
   }
   Flap Node!(ComponentPose:Component Space Pose Link Structure, Alpha:float=1) -> (Pose:Component Space Pose Link Structure) {
    is=title FortAnimGraphNode_Flap
    alphaBoolBlend {blendOption=Linear}
    sourceDrives [{
      bUseRange=false
      drivingCurve=null
      multiplier=1
      rangeMax=1
      rangeMin=-1
      remappedMax=1
      remappedMin=0
      sourceComponent=None
      useQuaternion=true
      sourceBone {boneName=None}
     }]
   }
   SelectFloat(A=0, B=1, bPickA=Get bIsCrouching())
  ]}
 EventGraph {nodes [
   Event BlueprintUpdateAnimation()
   n1 = GetPartSkeletalMeshComponent()
   > Set hand_l_dist_to_skirt(hand_l_dist_to_skirt:double=float - float(B=Get Distance Between Two Sockets!(
      Component:Skeletal Mesh Component Object Reference=@n1
      SocketOrBoneNameA:Name=dyn_skirt_B_01_r
      SocketSpaceA:ERelativeTransformSpace Enum=RTS_Component
      SocketOrBoneNameB:Name=hand_r
      SocketSpaceB:ERelativeTransformSpace Enum=RTS_Component
      bRemapRange:bool=true
      InRangeMin:float=16
      InRangeMax:float=22
      OutRangeMin:float=0
      OutRangeMax:float=1
     ) -> (ReturnValue:float) {is=title K2Node_CallFunction})) -> (Output_Get:double)
   > n2 = Branch(Condition=Get isWearingBackbling())
   > n3 = Get Dynamic Material Instance(SkeletalMeshComponent=GetPartSkeletalMeshComponent())
   > n4 = SetScalarParameterValue(self=@n3, ParameterName=Hide Element 01, Value=1)
   Do Once(execute=[@n4.then, @n2.else])
   > n5 = Cast To FortPlayerPawn(Object=TryGetPawnOwner())
   > Is Valid(InputObject=GetSkeletalMeshForPartType(self=@n5, PartType=Backpack))
   > Set isWearingBackbling(isWearingBackbling=true)
   n6 = GetOwnerSkeletalMeshComponent()
   Set left_thigh_rotation(left_thigh_rotation:Rotator=RInterpTo!(
     Current:Rotator=Get left_thigh_rotation()
     Target:Rotator=GetSocketRotation!(self:Scene Component Object Reference=@n6, InSocketName:Name=thigh_l) -> (ReturnValue:Rotator)
     DeltaTime:float=0
     InterpSpeed:float=10
    ) -> (ReturnValue:Rotator)) -> (Output_Get:Rotator)
   > Set right_thigh_rotation(right_thigh_rotation:Rotator=RInterpTo(Current=Get right_thigh_rotation(), Target=GetSocketRotation(self=@n6, InSocketName=thigh_r), DeltaTime=0, InterpSpeed=10)) -> (Output_Get:Rotator)
  ]}
]
variables [
 hand_l_dist_to_skirt {type=number}
 isWearingBackbling {type=boolean}
 left_thigh_rotation {type=Rotator}
 right_thigh_rotation {type=Rotator}
]
