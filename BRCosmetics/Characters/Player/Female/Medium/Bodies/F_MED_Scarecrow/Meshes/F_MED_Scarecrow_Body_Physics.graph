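# Physics asset for the body mesh of the female medium Scarecrow player cosmetic, with bodies on the pelvis and both thighs.
# Only the pelvis lists a collision shape, a single capsule, while the two thigh bodies have no shapes.
# Each thigh is constrained to the pelvis with free swing and twist, so the stored angular limits do not restrict motion.
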PhysicsAsset /BRCosmetics/Characters/Player/Female/Medium/Bodies/F_MED_Scarecrow/Meshes/F_MED_Scarecrow_Body_Physics
bodies {pelvis [{shapeType=Capsule, length=8.1388, radius=9.7832, center=<-3.5517 -1.0725 0>}], thigh_l [], thigh_r []}
constraints [
 {
  bone1Name=thigh_l
  bone2Name=pelvis
  swing1LimitDegrees=45
  swing1Motion=Free
  swing2LimitDegrees=45
  swing2Motion=Free
  twistLimitDegrees=45
  twistMotion=Free
 }
 {
  bone1Name=thigh_r
  bone2Name=pelvis
  swing1LimitDegrees=45
  swing1Motion=Free
  swing2LimitDegrees=45
  swing2Motion=Free
  twistLimitDegrees=45
  twistMotion=Free
 }
]
