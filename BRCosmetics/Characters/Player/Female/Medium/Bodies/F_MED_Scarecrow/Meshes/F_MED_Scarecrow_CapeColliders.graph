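# Physics asset of collision shapes for the female medium Scarecrow character body, named as colliders for its cape.
# Capsules cover the head, clavicles, pelvis, spine and feet, with paired capsules on the middle spine bones, one of each pair rotated crosswise.
# Large capsules on four dynamic skirt bones add skirt volume, while arm, thigh, calf and neck bodies carry no shapes.
# No constraints link the bodies.
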PhysicsAsset /BRCosmetics/Characters/Player/Female/Medium/Bodies/F_MED_Scarecrow/Meshes/F_MED_Scarecrow_CapeColliders
bodies {
 calf_l []
 calf_r []
 clavicle_l [{shapeType=Capsule, length=6.4464, radius=10.6353, center=<8.7872 -0.5227 1.6964>, rotation=<rot -82.1569 17.0975 167.7347>}]
 clavicle_r [{shapeType=Capsule, length=5.6964, radius=10.2957, center=<-7.8903 0.6286 -2.6686>, rotation=<rot -82.157 17.0976 -12.2654>}]
 dyn_skirt_B_01_l [{shapeType=Capsule, length=15, radius=15, shapeName=dyn_skirt_B_01_l_capsule, center=<18.0353 -3.9452 0>, rotation=<rot 90 180 180>}]
 dyn_skirt_B_01_r [{shapeType=Capsule, length=15, radius=15, shapeName=dyn_skirt_B_01_r_capsule, center=<18.0353 -3.9452 0>, rotation=<rot 90 179.9901 179.9901>}]
 dyn_skirt_C_01_l [{shapeType=Capsule, length=2, radius=24, shapeName=dyn_skirt_C_01_l_capsule, center=<21.4885 7.876 2.8335>, rotation=<rot 0 0.0004 0>}]
 dyn_skirt_C_01_r [{shapeType=Capsule, length=2.25, radius=24, shapeName=dyn_skirt_C_01_r_capsule, center=<21.4885 7.876 -0.8685>, rotation=<rot 0 0.0004 0>}]
 foot_l [{shapeType=Capsule, length=12.5982, radius=6.5127, center=<-2.2072 6.0534 -0.3367>, rotation=<rot 0.0388 1.4701 -89.3982>}]
 foot_r [{shapeType=Capsule, length=12.5982, radius=6.5127, center=<2.2072 -6.0534 0.3367>, rotation=<rot 0.0388 1.4701 90.6018>}]
 head [{shapeType=Capsule, length=0, radius=18, center=<9.0697 0.0746 0>, rotation=<rot 0 -90 0>}]
 lowerarm_l []
 lowerarm_r []
 neck_02 []
 pelvis [{shapeType=Capsule, length=12.1528, radius=14.6596, center=<-1.9544 -2.2029 -0.1981>}]
 spine_01 [{shapeType=Capsule, length=6.2671, radius=12.9854, center=<-2.8183 -1.0873 0>}]
 spine_02 [
  {shapeType=Capsule, length=11.5484, radius=10.4854, center=<1.179 -1.7706 0>}
  {shapeType=Capsule, length=20.6875, radius=9.3125, center=<-1.8317 2.606 0>, rotation=<rot 90 0 0>}
 ]
 spine_03 [
  {shapeType=Capsule, length=9.5484, radius=9.9854, center=<3.3444 -1.1726 0>}
  {shapeType=Capsule, length=20.7188, radius=9.2812, center=<-1.2326 2.6059 0>, rotation=<rot 90 0 0>}
 ]
 spine_04 [
  {shapeType=Capsule, length=7.3296, radius=13.8916, center=<4.5142 -4.5745 0>}
  {shapeType=Capsule, length=10.0312, radius=9.2812, center=<-1.2326 1.308 0>, rotation=<rot 90 0 0>}
 ]
 spine_05 [{shapeType=Capsule, length=10.9859, radius=10.0791, center=<5.6377 0.3227 0>}]
 thigh_l []
 thigh_r []
 upperarm_l []
 upperarm_r []
}
constraints []
