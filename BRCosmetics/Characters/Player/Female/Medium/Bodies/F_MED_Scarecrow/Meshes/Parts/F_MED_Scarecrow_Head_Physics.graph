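# Physics asset for the head mesh of the female medium Scarecrow character, rigging its hair bones for physical motion.
# Capsule bodies cover the head, two central hair bones and left and right strands, with each hair body constrained directly to the head rather than chained.
# The side strands lock twist and one swing axis to pivot on a single axis, while the central hair bones have swing and twist limits.
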PhysicsAsset /BRCosmetics/Characters/Player/Female/Medium/Bodies/F_MED_Scarecrow/Meshes/Parts/F_MED_Scarecrow_Head_Physics
bodies {
 dyn_hair_01 [{shapeType=Capsule, length=2.1, radius=3.5, center=<1.3219 -1.3256 0>}]
 dyn_hair_02 [{shapeType=Capsule, length=2.1, radius=3.5, center=<1.3219 -1.3256 0>}]
 dyn_hair_strand_l [{shapeType=Capsule, length=5, radius=2, center=<3 0 0>, rotation=<rot 90 0 0>}]
 dyn_hair_strand_r [{shapeType=Capsule, length=5, radius=2, center=<-3 0 0>, rotation=<rot 90 180 0>}]
 head [{shapeType=Capsule, length=5.5922, radius=5.6714, center=<5.054 0.3532 0>, rotation=<rot -90 0 0>}]
}
constraints [
 {
  bone1Name=dyn_hair_strand_l
  bone2Name=head
  swing1LimitDegrees=25
  swing2LimitDegrees=45
  swing2Motion=Locked
  twistLimitDegrees=45
  twistMotion=Locked
 }
 {
  bone1Name=dyn_hair_strand_r
  bone2Name=head
  swing1LimitDegrees=25
  swing2LimitDegrees=45
  swing2Motion=Locked
  twistLimitDegrees=45
  twistMotion=Locked
 }
 {bone1Name=dyn_hair_02, bone2Name=head, swing1LimitDegrees=40, swing2LimitDegrees=40, twistLimitDegrees=10}
 {bone1Name=dyn_hair_01, bone2Name=head, swing1LimitDegrees=40, swing2LimitDegrees=40, twistLimitDegrees=10}
]
